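# Physics asset for the shoelaces on the F_MED_Football_20 female player body, set up to let the lace bones swing against the feet.
# Box bodies cover two lace bones per shoe plus one larger box on each foot bone.
# Each lace is constrained to its same-side foot with a narrow swing range and locked twist.
# The right-side lace bones are spelled "shoerace" rather than "shoelace".
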PhysicsAsset /BRCosmetics/Characters/Player/Female/Medium/Bodies/F_MED_Football_20/Meshes/F_MED_Football_20_shoelaces
bodies {
 dyn_shoelace_01_l [{shapeType=Box, extentX=1.3935, extentY=0.9035, extentZ=4.4236, center=<2.215 0.2859 -0.4181>, rotation=<rot 0 -80.9574 -73.1431>}]
 dyn_shoelace_02_l [{shapeType=Box, extentX=1.2527, extentY=0.9301, extentZ=4.498, center=<2.0405 -0.28 0.2075>, rotation=<rot 0 84.9295 -81.4042>}]
 dyn_shoerace_01_r [{shapeType=Box, extentX=1.2439, extentY=0.9682, extentZ=4.498, center=<1.5756 -0.0005 2.3356>, rotation=<rot 0 88.7423 -58.6849>}]
 dyn_shoerace_02_r [{shapeType=Box, extentX=1.3874, extentY=0.9358, extentZ=4.4236, center=<1.3655 -0.0966 -2.1751>, rotation=<rot 0 -87.9207 -65.4531>}]
 foot_l [{shapeType=Box, extentX=10.7037, extentY=9.3847, extentZ=20.206, center=<-4.3397 5.26 -0.3151>, rotation=<rot 0 33.9172 -88.9893>}]
 foot_r [{shapeType=Box, extentX=9.2037, extentY=9.3847, extentZ=20.206, center=<2.2441 -3.2896 -0.1252>, rotation=<rot 0 33.9162 -88.9893>}]
}
constraints [
 {bone1Name=dyn_shoelace_01_l, bone2Name=foot_l, swing1LimitDegrees=33, swing2LimitDegrees=33, twistLimitDegrees=45, twistMotion=Locked}
 {bone1Name=dyn_shoelace_02_l, bone2Name=foot_l, swing1LimitDegrees=33, swing2LimitDegrees=33, twistLimitDegrees=45, twistMotion=Locked}
 {bone1Name=dyn_shoerace_01_r, bone2Name=foot_r, swing1LimitDegrees=33, swing2LimitDegrees=33, twistLimitDegrees=45, twistMotion=Locked}
 {bone1Name=dyn_shoerace_02_r, bone2Name=foot_r, swing1LimitDegrees=33, swing2LimitDegrees=33, twistLimitDegrees=45, twistMotion=Locked}
]
