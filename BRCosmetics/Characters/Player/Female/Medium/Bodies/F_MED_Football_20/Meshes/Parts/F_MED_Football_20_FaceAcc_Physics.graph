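# Physics asset for the face accessory of the F_MED_Football_20 female character body, covering four dynamic strap bones around the head.
# Each top and bottom strap on both sides uses a small box body, while the head uses a single sphere.
# Each strap is constrained directly to the head with twist and first swing at zero degrees, leaving a narrow second-swing range, slightly wider for the bottom straps.
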PhysicsAsset /BRCosmetics/Characters/Player/Female/Medium/Bodies/F_MED_Football_20/Meshes/Parts/F_MED_Football_20_FaceAcc_Physics
bodies {
 dyn_strap_bottom_l [{shapeType=Box, extentX=3, extentY=1, extentZ=1.5, center=<1 0 0>}]
 dyn_strap_bottom_r [{shapeType=Box, extentX=3, extentY=1, extentZ=1.5, center=<1 0 0>}]
 dyn_strap_top_l [{shapeType=Box, extentX=2.6753, extentY=1, extentZ=1.3463, center=<1 0 0>}]
 dyn_strap_top_r [{shapeType=Box, extentX=2.6275, extentY=1, extentZ=1.3463, center=<1 0 0>}]
 head [{shapeType=Sphere, radius=9.8259, center=<7.4953 -0.4088 -0.0043>}]
}
constraints [
 {bone1Name=dyn_strap_bottom_l, bone2Name=head, swing1LimitDegrees=0, swing2LimitDegrees=6, twistLimitDegrees=0}
 {bone1Name=dyn_strap_top_l, bone2Name=head, swing1LimitDegrees=0, swing2LimitDegrees=4, twistLimitDegrees=0}
 {bone1Name=dyn_strap_top_r, bone2Name=head, swing1LimitDegrees=0, swing2LimitDegrees=4, twistLimitDegrees=0}
 {bone1Name=dyn_strap_bottom_r, bone2Name=head, swing1LimitDegrees=0, swing2LimitDegrees=6, twistLimitDegrees=0}
]
